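; Splitenergy vehicle boost cosmetic actor with a skeletal mesh, which it offsets at BeginPlay through the cosmetics function library.
; Boost start and finish stop the opposing timeline and play Fade In or Fade Out from the start, driving a dissolve parameter on the mesh materials.
; Fade Out completion triggers a parent call, and boost start also calls its parent.
; When color is ready, the painted variant ID is applied to the mesh materials.
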(event EventBeginPlay
  ; Set initial mesh location
  (Class|BPVehicleCosmeticsFunctionLibrary|CosmeticActorSetSkeletalAnimLocation self "-75.000000,0.000000,-4.000000" (Variables|Default|GetSkeletalMesh)))

(event Collision|EventActorBeginOverlap
  (params (OtherActor :type Actor%20Object%20Reference)))

(event EventTick
  (params (DeltaSeconds :type Float%20%28single-precision%29)))

(event Vehicle|Cosmetics|EventOnBoostStarted
  ; Update Dissolve on SKMesh
  (Utilities|FlowControl|Sequence
    (:then_0
      (bind (_direction _dissolve) (Timeline|Fade%20Out%20Timeline Stop)
        (:Update
          (Rendering|Material|SetScalarParameterValueonMaterials (Variables|Default|GetSkeletalMesh) (Variables|Default|GetDissolveParameterName) _dissolve))
        (:Finished
          (|CallParentFunction))))
    (:then_1
      (|CallParentFunction)
      (bind (_direction_2 _dissolve_2) (Timeline|Fade%20In%20Timeline PlayFromStart)
        (:Update
          (Rendering|Material|SetScalarParameterValueonMaterials (Variables|Default|GetSkeletalMesh) (Variables|Default|GetDissolveParameterName) _dissolve_2))))))

(event Vehicle|Cosmetics|EventOnBoostFinished
  ; Update Dissolve on SKMesh
  (Utilities|FlowControl|Sequence
    (:then_0
      (Timeline|Fade%20In%20Timeline Stop))
    (:then_1
      (Timeline|Fade%20Out%20Timeline PlayFromStart))))

(event Vehicle|Cosmetics|EventOnColorReady
  (params (Color :type Name :default "None") (AdditionalInfo :type Vehicle%20Cosmetics%20Additional%20Parameter%20Infos%20Structure))
  ; Update Painted variants
  (|CallParentFunction Color AdditionalInfo)
  (Class|BPVehicleCosmeticsFunctionLibrary|CosmeticActorSetColorVariantIDToMaterials Color self (Variables|Default|GetSkeletalMesh)))

(fn ConstructionScript
  (|CallParentFunction))
