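# Niagara system for the Splitenergy vehicle boost cosmetic: a sprite cone glow and a layered ribbon trail behind the car.
# Both emitters offset by car and boost position user parameters, with per-paint color overrides; cone colors and the third ribbon layer pulse via sine-driven array lookups.
# The ribbon spawns by quality level, feeds four renderers, widens and fades over life, inherits decaying source movement, and moves faster backward in frontend preview.
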NiagaraSystem /VehicleCosmetics/Boosts/Boost_Splitenergy/FX/NS_Boost_Splitenergy
dependencies {
 usedDataInterfaces [
  &NiagaraDataInterfaceArrayColor
  &NiagaraDataInterfaceCurve
  &NiagaraDataInterfaceArrayFloat
  &NiagaraDataInterfaceSkeletalMesh
 ]
 usedDynamicInputs []
 usedModules [
  &self:SystemSpawnScript.SystemScriptSource.SystemScriptGraph.NiagaraNodeAssignment_0.SetVariables_#1
  &SystemState
 ]
}
emitters {
 NE_VC_BoostCone {
  emitterSpawnScript [/SetVariables(Emitter.Emissive:NiagaraFloat=&SelectFloatFromArray(
     Direct Array Index=$System.ColorVariantID
     Float Selection Array=DataInterface() {propertyValues {FloatData [3,1,2,0.5,2,1,2,4,2,1,0.8,2,1,2,3]}}
    )) {isSetParameters=true}]
  emitterUpdateScript [
   &EmitterState(Inactive Response=::NewEnumerator2, Life Cycle Mode=::NewEnumerator1, Loop Duration=1.5, UseLoopDelay=false)
   &VC_FilteredSocketCount() {enabled=false}
   &SpawnRate(SpawnRate=3, Use Spawn Probability=false)
   &SpawnBurst_Instantaneous(Spawn Group=1, Use Spawn Probability=false, UseLoopCountLimit=false)
  ]
  particleSpawnScript [
   &VC_PaintedVariantsOverride(
    Black=&SelectLinearColorFromArray(
     Array Sampling Mode=::NewEnumerator2
     Direct Array Index=&MakeIntFromFloat(Float=&Sine(Normalized Angle=$Emitter.Age, Period=0.5, Scale=1, Bias=0), Conversion Mode=::NewEnumerator1)
     Direct Array Mode=::NewEnumerator1
     Interpolation=&Sine(Normalized Angle=$Emitter.Age, Period=0.5, Scale=1, Bias=0)
     Color Selection Array=DataInterface() {propertyValues {ColorData [<rgba 0.0365 0.0365 0.0365 0.5>, <rgba 0.0365 0.0365 0.0365 0.5>]}}
     Random Seed=?
    )
    BurntSienna=&SelectLinearColorFromArray(
     Array Sampling Mode=::NewEnumerator2
     Direct Array Index=&MakeIntFromFloat(Float=&Sine(Normalized Angle=$Emitter.Age, Period=0.5, Scale=1, Bias=0), Conversion Mode=::NewEnumerator1)
     Direct Array Mode=::NewEnumerator1
     Interpolation=&Sine(Normalized Angle=$Emitter.Age, Period=0.5, Scale=1, Bias=0)
     Color Selection Array=DataInterface() {propertyValues {ColorData [<rgba 0.2292 0.0462 0 2.5>, <rgba 1 0.2018 0 0.08>]}}
     Random Seed=?
    )
    Cobalt=&SelectLinearColorFromArray(
     Array Sampling Mode=::NewEnumerator2
     Direct Array Index=&MakeIntFromFloat(Float=&Sine(Normalized Angle=$Emitter.Age, Period=0.5, Scale=1, Bias=0), Conversion Mode=::NewEnumerator1)
     Direct Array Mode=::NewEnumerator1
     Interpolation=&Sine(Normalized Angle=$Emitter.Age, Period=0.5, Scale=1, Bias=0)
     Color Selection Array=DataInterface() {propertyValues {ColorData [<rgba 0 0.0185 1 4>, <rgba 0 0.0185 1 5>]}}
     Random Seed=?
    )
    Crimson=&SelectLinearColorFromArray(
     Array Sampling Mode=::NewEnumerator2
     Direct Array Index=&MakeIntFromFloat(Float=&Sine(Normalized Angle=$Emitter.Age, Period=0.5, Scale=1, Bias=0), Conversion Mode=::NewEnumerator1)
     Direct Array Mode=::NewEnumerator1
     Interpolation=&Sine(Normalized Angle=$Emitter.Age, Period=0.5, Scale=1, Bias=0)
     Color Selection Array=DataInterface() {propertyValues {ColorData [<rgba 1 0.0417 0.0676 10>, <rgba 1 0.0417 0.0676 25>]}}
     Random Seed=?
    )
    DefaultColor=&SelectLinearColorFromArray(
     Array Sampling Mode=::NewEnumerator2
     Direct Array Index=&MakeIntFromFloat(Float=&Sine(Normalized Angle=$Emitter.Age, Period=0.5, Scale=1, Bias=0), Conversion Mode=::NewEnumerator1)
     Direct Array Mode=::NewEnumerator1
     Interpolation=&Sine(Normalized Angle=$Emitter.Age, Period=0.5, Scale=1, Bias=0)
     Color Selection Array=DataInterface() {propertyValues {ColorData [<rgba 0.02 0.2265 1 0.25>, <rgba 1 0.149 0.498 0.25>]}}
     Random Seed=?
    )
    ForestGreen=&SelectLinearColorFromArray(
     Array Sampling Mode=::NewEnumerator2
     Direct Array Index=&MakeIntFromFloat(Float=&Sine(Normalized Angle=$Emitter.Age, Period=0.5, Scale=1, Bias=0), Conversion Mode=::NewEnumerator1)
     Direct Array Mode=::NewEnumerator1
     Interpolation=&Sine(Normalized Angle=$Emitter.Age, Period=0.5, Scale=1, Bias=0)
     Color Selection Array=DataInterface() {propertyValues {ColorData [<rgba 0 0.349 0.0213 55>, <rgba 0 0.349 0.0213 60>]}}
     Random Seed=?
    )
    Grey=&SelectLinearColorFromArray(
     Array Sampling Mode=::NewEnumerator2
     Direct Array Index=&MakeIntFromFloat(Float=&Sine(Normalized Angle=$Emitter.Age, Period=0.5, Scale=1, Bias=0), Conversion Mode=::NewEnumerator1)
     Direct Array Mode=::NewEnumerator1
     Interpolation=&Sine(Normalized Angle=$Emitter.Age, Period=0.5, Scale=1, Bias=0)
     Color Selection Array=DataInterface() {propertyValues {ColorData [<rgba 0.0156 0.0156 0.0156 55>, <rgba 0.0365 0.0365 0.0365 65>]}}
     Random Seed=?
    )
    Lime=&SelectLinearColorFromArray(
     Array Sampling Mode=::NewEnumerator2
     Direct Array Index=&MakeIntFromFloat(Float=&Sine(Normalized Angle=$Emitter.Age, Period=0.5, Scale=1, Bias=0), Conversion Mode=::NewEnumerator1)
     Direct Array Mode=::NewEnumerator1
     Interpolation=&Sine(Normalized Angle=$Emitter.Age, Period=0.5, Scale=1, Bias=0)
     Color Selection Array=DataInterface() {propertyValues {ColorData [<rgba 0.1458 0.2917 0 55>, <rgba 0.1458 0.2917 0 65>]}}
     Random Seed=?
    )
    Orange=&SelectLinearColorFromArray(
     Array Sampling Mode=::NewEnumerator2
     Direct Array Index=&MakeIntFromFloat(Float=&Sine(Normalized Angle=$Emitter.Age, Period=0.5, Scale=1, Bias=0), Conversion Mode=::NewEnumerator1)
     Direct Array Mode=::NewEnumerator1
     Interpolation=&Sine(Normalized Angle=$Emitter.Age, Period=0.5, Scale=1, Bias=0)
     Color Selection Array=DataInterface() {propertyValues {ColorData [<rgba 0.6724 0.0643 0.0002 5>, <rgba 0.6724 0.0643 0.0002 65>]}}
     Random Seed=?
    )
    Pink=&SelectLinearColorFromArray(
     Array Sampling Mode=::NewEnumerator2
     Direct Array Index=&MakeIntFromFloat(Float=&Sine(Normalized Angle=$Emitter.Age, Period=0.5, Scale=1, Bias=0), Conversion Mode=::NewEnumerator1)
     Direct Array Mode=::NewEnumerator1
     Interpolation=&Sine(Normalized Angle=$Emitter.Age, Period=0.5, Scale=1, Bias=0)
     Color Selection Array=DataInterface() {propertyValues {ColorData [<rgba 1 0.1771 0.8521 50>, <rgba 1 0.1771 0.8521 50>]}}
     Random Seed=?
    )
    Purple=&SelectLinearColorFromArray(
     Array Sampling Mode=::NewEnumerator2
     Direct Array Index=&MakeIntFromFloat(Float=&Sine(Normalized Angle=$Emitter.Age, Period=0.5, Scale=1, Bias=0), Conversion Mode=::NewEnumerator1)
     Direct Array Mode=::NewEnumerator1
     Interpolation=&Sine(Normalized Angle=$Emitter.Age, Period=0.5, Scale=1, Bias=0)
     Color Selection Array=DataInterface() {propertyValues {ColorData [<rgba 0.0995 0 0.6795 55>, <rgba 0.0995 0 0.6795 65>]}}
     Random Seed=?
    )
    Saffron=&SelectLinearColorFromArray(
     Array Sampling Mode=::NewEnumerator2
     Direct Array Index=&MakeIntFromFloat(Float=&Sine(Normalized Angle=$Emitter.Age, Period=0.5, Scale=1, Bias=0), Conversion Mode=::NewEnumerator1)
     Direct Array Mode=::NewEnumerator1
     Interpolation=&Sine(Normalized Angle=$Emitter.Age, Period=0.5, Scale=1, Bias=0)
     Color Selection Array=DataInterface() {propertyValues {ColorData [<rgba 0.9 0.69 0 45>, <rgba 0.9 0.69 0 55>]}}
     Random Seed=?
    )
    SkyBlue=&SelectLinearColorFromArray(
     Array Sampling Mode=::NewEnumerator2
     Direct Array Index=&MakeIntFromFloat(Float=&Sine(Normalized Angle=$Emitter.Age, Period=0.5, Scale=1, Bias=0), Conversion Mode=::NewEnumerator1)
     Direct Array Mode=::NewEnumerator1
     Interpolation=&Sine(Normalized Angle=$Emitter.Age, Period=0.5, Scale=1, Bias=0)
     Color Selection Array=DataInterface() {propertyValues {ColorData [<rgba 0 0.357 0.5365 55>, <rgba 0 0.357 0.5365 65>]}}
     Random Seed=?
    )
    TitaniumWhite=&SelectLinearColorFromArray(
     Array Sampling Mode=::NewEnumerator2
     Direct Array Index=&MakeIntFromFloat(Float=&Sine(Normalized Angle=$Emitter.Age, Period=0.5, Scale=1, Bias=0), Conversion Mode=::NewEnumerator1)
     Direct Array Mode=::NewEnumerator1
     Interpolation=&Sine(Normalized Angle=$Emitter.Age, Period=0.5, Scale=1, Bias=0)
     Color Selection Array=DataInterface() {propertyValues {ColorData [<rgba 0.75 0.75 0.75 0.2>, <rgba 0.75 0.75 0.75 0.6>]}}
     Random Seed=?
    )
   )
   /SetVariables(Particles.NewTagVis:NiagaraInt32=&SpawnGroup()) {isSetParameters=true}
   &InitializeParticle(
    Color=$StackContext.Color
    Lifetime=/VehicleCosmetics/FX/Niagara/Modules/MergeEmitters/VC_FloatBySpawnGroup(
     SpawnGroup or ExecIndex:NiagaraInt32=0
     CustomInt:NiagaraInt32=0
     MaximumSpawnGroups:NiagaraInt32=2
     Float SG 0:NiagaraFloat=1.5
     Float SG 1:NiagaraFloat=$Emitter.CurrentLoopDuration
     Float SG 2:NiagaraFloat=0
     Float SG 3:NiagaraFloat=0
     Float SG 4:NiagaraFloat=0
     Float SG 5:NiagaraFloat=0
    )
    Position Offset=&AddVectorToPosition(
     Position=<-75 0 -4>
     Vector=&AddVectorToPosition(Position=$User.DTCarOffset, Vector=Linked() {linkedVariable {name=User.SingleBoostPos, type=NiagaraPosition}})
    )
    Position Offset Coordinate Space=::Local
    Sprite Rotation Mode=::NewEnumerator2
    Sprite Size Mode=::NewEnumerator3
    Uniform Sprite Size=100
    UsePositionOffset=true
    Write Material Random=true
   )
   &PartitionParticles(Number Of Partitions=$Emitter.PortCount)
   &SkeletalMeshLocation(
    Bone / Socket Index=$Particles.PartitionParticles.Partition
    Bone / Socket Sampling Mode=::NewEnumerator5
    MaskSpawnGroup=false
    Mesh Sampling Type=::NewEnumerator6
    Skeletal Mesh=$System.SkeletalMesh
   ) {enabled=false}
   &AddVelocity(Velocity=<-1500 0 0>) {enabled=false}
  ]
  particleUpdateScript [
   &ParticleState()
   &ScaleSpriteSize(Uniform Curve Sprite Scale=DataInterface() {propertyValues {Curve {keys [
        <at 0 0 RCIM_Cubic>
        {arriveTangent=2, interpMode=RCIM_Cubic, leaveTangent=2, time=0.1, value=1}
        <at 0.4 0.8 RCIM_Cubic>
        {arriveTangent=0.2432, interpMode=RCIM_Cubic, leaveTangent=0.2432, time=0.58, value=1}
        <at 0.77 0.89 RCIM_Cubic>
        {arriveTangent=-3.8696, interpMode=RCIM_Cubic, leaveTangent=-3.8696, time=0.9, value=1}
        <at 1 0 RCIM_Cubic>
       ]}}})
   &SolveForcesAndVelocity(Clamp Velocity=false, Limit Acceleration=false)
   &ScaleColor(Scale Alpha=2, Scale RGB=<10 10 10>, ScaleA=true, ScaleRGB=true)
  ]
  rendererClasses [&NiagaraSpriteRendererProperties]
  renderers [{rendererClass=&NiagaraSpriteRendererProperties}]
 }
 NE_VC_Boost_Ribbon_Optimized {
  emitterUpdateScript [
   &EmitterState()
   &VC_FilteredSocketCount() {enabled=false}
   &VC_RateQualityHandler(Epic Spawn Rate=120, High Spawn Rate=80, Low Spawn Rate=40, Medium Spawn Rate=60, Use Custom Rate?=true)
   &VC_SpawnRibbon_v1(SpawnRate=60)
  ]
  particleSpawnScript [
   &VC_PaintedVariantsOverride(
    Black=<rgba 0.1878 0.1878 0.1878 5>
    BurntSienna=<rgba 0.5417 0.296 0.1749 5>
    Cobalt=<rgba 0.3844 0.4666 1.8 5>
    Crimson=<rgba 1 0.1458 0.169 5>
    DefaultColor=<rgba 0.2582 0.2582 0.9734 4>
    ForestGreen=<rgba 0 1.2 0.0606 5>
    Grey=<rgba 0.5677 0.5677 0.5677 3.5>
    Lime=<rgba 0.868 1.1 0.6359 5>
    Orange=<rgba 1.1 0.3837 0 5>
    Pink=<rgba 1.0742 0.6042 0.9781 2>
    Purple=<rgba 0.4104 0 1 3.5>
    R1 Black:LinearColor=<rgba 0.0191 0.0191 0.0191 7>
    R1 BurntSienna:LinearColor=<rgba 0.526 0.3202 0.2329 8>
    R1 Cobalt:LinearColor=<rgba 0.1912 0.2384 0.8308 6.5>
    R1 Crimson:LinearColor=<rgba 0.75 0.0469 0.0659 8>
    R1 DefaultColor=<rgba 0.8333 0.217 0.8333 8>
    R1 ForestGreen:LinearColor=<rgba 0 0.8438 0.0426 7>
    R1 Grey:LinearColor=<rgba 0.0885 0.0885 0.0885 6>
    R1 Lime:LinearColor=<rgba 0.8341 1 0 6>
    R1 Orange:LinearColor=<rgba 0.8333 0.2907 0 7>
    R1 Pink:LinearColor=<rgba 0.7448 0.3802 0.7202 7>
    R1 Purple:LinearColor=<rgba 0.35 0.198 0.75 7>
    R1 Saffron:LinearColor=<rgba 0.8073 0.7711 0.0505 7>
    R1 SkyBlue:LinearColor=<rgba 0.2472 1.0052 1.1 7>
    R1 TitaniumWhite:LinearColor=<rgba 0.75 0.75 0.75 7>
    R2 Black:LinearColor=<rgba 0.1146 0.1146 0.1146 4>
    R2 BurntSienna:LinearColor=<rgba 0.3594 0.2062 0.1479 4>
    R2 Cobalt:LinearColor=<rgba 0.2979 0.3165 1.3 4>
    R2 Crimson:LinearColor=<rgba 0.5885 0.0276 0.0428 4>
    R2 DefaultColor:LinearColor=<rgba 0.0525 0.2625 1.05 3>
    R2 ForestGreen:LinearColor=<rgba 0 0.7865 0.0397 4>
    R2 Grey:LinearColor=<rgba 0.375 0.375 0.375 4>
    R2 Lime:LinearColor=<rgba 0.4721 0.5885 0.3556 4>
    R2 Orange:LinearColor=<rgba 0.6724 0.2346 0 4>
    R2 Pink:LinearColor=<rgba 0.5588 0.3617 0.6146 4>
    R2 Purple:LinearColor=<rgba 0.2978 0.0668 0.8333 4>
    R2 Saffron:LinearColor=<rgba 0.6094 0.5669 0.2603 4>
    R2 SkyBlue:LinearColor=<rgba 0.129 0.6115 0.8542 4>
    R2 TitaniumWhite:LinearColor=<rgba 0.5573 0.5573 0.5573 4>
    R3 Black:LinearColor=&SelectLinearColorFromArray(
     Array Sampling Mode=::NewEnumerator2
     Direct Array Index=&MakeIntFromFloat(Float=&Sine(Normalized Angle=$Emitter.Age, Period=0.5, Scale=1, Bias=0), Conversion Mode=::NewEnumerator1)
     Direct Array Mode=::NewEnumerator1
     Interpolation=&Sine(Normalized Angle=$Emitter.Age, Period=0.5, Scale=1, Bias=0)
     Color Selection Array=DataInterface() {propertyValues {ColorData [<rgba 0.8906 0.8906 0.8906 0.5>, <rgba 1 1 1 1.5>]}}
     Random Seed=?
    )
    R3 BurntSienna:LinearColor=&SelectLinearColorFromArray(
     Array Sampling Mode=::NewEnumerator2
     Direct Array Index=&MakeIntFromFloat(Float=&Sine(Normalized Angle=$Emitter.Age, Period=0.5, Scale=1, Bias=0), Conversion Mode=::NewEnumerator1)
     Direct Array Mode=::NewEnumerator1
     Interpolation=&Sine(Normalized Angle=$Emitter.Age, Period=0.5, Scale=1, Bias=0)
     Color Selection Array=DataInterface() {propertyValues {ColorData [<rgba 0.8177 0.2525 0.1008 0.8>, <rgba 0.8177 0.2525 0.1008 1>]}}
     Random Seed=?
    )
    R3 Cobalt:LinearColor=&SelectLinearColorFromArray(
     Array Sampling Mode=::NewEnumerator2
     Direct Array Index=&MakeIntFromFloat(Float=&Sine(Normalized Angle=$Emitter.Age, Period=0.5, Scale=1, Bias=0), Conversion Mode=::NewEnumerator1)
     Direct Array Mode=::NewEnumerator1
     Interpolation=&Sine(Normalized Angle=$Emitter.Age, Period=0.5, Scale=1, Bias=0)
     Color Selection Array=DataInterface() {propertyValues {ColorData [<rgba 0.75 0.89 3 0.5>, <rgba 0.75 0.89 3 0.75>]}}
     Random Seed=?
    )
    R3 Crimson:LinearColor=&SelectLinearColorFromArray(
     Array Sampling Mode=::NewEnumerator2
     Direct Array Index=&MakeIntFromFloat(Float=&Sine(Normalized Angle=$Emitter.Age, Period=0.5, Scale=1, Bias=0), Conversion Mode=::NewEnumerator1)
     Direct Array Mode=::NewEnumerator1
     Interpolation=&Sine(Normalized Angle=$Emitter.Age, Period=0.5, Scale=1, Bias=0)
     Color Selection Array=DataInterface() {propertyValues {ColorData [<rgba 1 0.099 0.1234 1>, <rgba 1 0.099 0.1234 1.5>]}}
     Random Seed=?
    )
    R3 DefaultColor:LinearColor=&SelectLinearColorFromArray(
     Array Sampling Mode=::NewEnumerator2
     Direct Array Index=&MakeIntFromFloat(Float=&Sine(Normalized Angle=$Emitter.Age, Period=0.5, Scale=1, Bias=0), Conversion Mode=::NewEnumerator1)
     Direct Array Mode=::NewEnumerator1
     Interpolation=&Sine(Normalized Angle=$Emitter.Age, Period=0.5, Scale=1, Bias=0)
     Color Selection Array=DataInterface() {propertyValues {ColorData [<rgba 0.8906 0.8906 0.8906 0.8>, <rgba 1 1 1 1.6>]}}
     Random Seed=?
    )
    R3 ForestGreen:LinearColor=&SelectLinearColorFromArray(
     Array Sampling Mode=::NewEnumerator2
     Direct Array Index=&MakeIntFromFloat(Float=&Sine(Normalized Angle=$Emitter.Age, Period=0.5, Scale=1, Bias=0), Conversion Mode=::NewEnumerator1)
     Direct Array Mode=::NewEnumerator1
     Interpolation=&Sine(Normalized Angle=$Emitter.Age, Period=0.5, Scale=1, Bias=0)
     Color Selection Array=DataInterface() {propertyValues {ColorData [<rgba 0 2 0.1011 0.5>, <rgba 0 2 0.1011 1>]}}
     Random Seed=?
    )
    R3 Grey:LinearColor=&SelectLinearColorFromArray(
     Array Sampling Mode=::NewEnumerator2
     Direct Array Index=&MakeIntFromFloat(Float=&Sine(Normalized Angle=$Emitter.Age, Period=0.5, Scale=1, Bias=0), Conversion Mode=::NewEnumerator1)
     Direct Array Mode=::NewEnumerator1
     Interpolation=&Sine(Normalized Angle=$Emitter.Age, Period=0.5, Scale=1, Bias=0)
     Color Selection Array=DataInterface() {propertyValues {ColorData [<rgba 0.8906 0.8906 0.8906 0.25>, <rgba 1 1 1 0.5>]}}
     Random Seed=?
    )
    R3 Lime:LinearColor=&SelectLinearColorFromArray(
     Array Sampling Mode=::NewEnumerator2
     Direct Array Index=&MakeIntFromFloat(Float=&Sine(Normalized Angle=$Emitter.Age, Period=0.5, Scale=1, Bias=0), Conversion Mode=::NewEnumerator1)
     Direct Array Mode=::NewEnumerator1
     Interpolation=&Sine(Normalized Angle=$Emitter.Age, Period=0.5, Scale=1, Bias=0)
     Color Selection Array=DataInterface() {propertyValues {ColorData [<rgba 0.8151 1 0.6302 1>, <rgba 0.8151 1 0.6302 2>]}}
     Random Seed=?
    )
    R3 Orange:LinearColor=&SelectLinearColorFromArray(
     Array Sampling Mode=::NewEnumerator2
     Direct Array Index=&MakeIntFromFloat(Float=&Sine(Normalized Angle=$Emitter.Age, Period=0.5, Scale=1, Bias=0), Conversion Mode=::NewEnumerator1)
     Direct Array Mode=::NewEnumerator1
     Interpolation=&Sine(Normalized Angle=$Emitter.Age, Period=0.5, Scale=1, Bias=0)
     Color Selection Array=DataInterface() {propertyValues {ColorData [<rgba 1.1 0.3837 0 1>, <rgba 1.1 0.3837 0 1.5>]}}
     Random Seed=?
    )
    R3 Pink:LinearColor=&SelectLinearColorFromArray(
     Array Sampling Mode=::NewEnumerator2
     Direct Array Index=&MakeIntFromFloat(Float=&Sine(Normalized Angle=$Emitter.Age, Period=0.5, Scale=1, Bias=0), Conversion Mode=::NewEnumerator1)
     Direct Array Mode=::NewEnumerator1
     Interpolation=&Sine(Normalized Angle=$Emitter.Age, Period=0.5, Scale=1, Bias=0)
     Color Selection Array=DataInterface() {propertyValues {ColorData [<rgba 1.1719 0.6714 0.9495 1>, <rgba 1.1719 0.6714 0.9495 1.2>]}}
     Random Seed=?
    )
    R3 Purple:LinearColor=&SelectLinearColorFromArray(
     Array Sampling Mode=::NewEnumerator2
     Direct Array Index=&MakeIntFromFloat(Float=&Sine(Normalized Angle=$Emitter.Age, Period=0.5, Scale=1, Bias=0), Conversion Mode=::NewEnumerator1)
     Direct Array Mode=::NewEnumerator1
     Interpolation=&Sine(Normalized Angle=$Emitter.Age, Period=0.5, Scale=1, Bias=0)
     Color Selection Array=DataInterface() {propertyValues {ColorData [<rgba 0.6162 0.349 1 0.85>, <rgba 0.6162 0.349 1 1>]}}
     Random Seed=?
    )
    R3 Saffron:LinearColor=&SelectLinearColorFromArray(
     Array Sampling Mode=::NewEnumerator2
     Direct Array Index=&MakeIntFromFloat(Float=&Sine(Normalized Angle=$Emitter.Age, Period=0.5, Scale=1, Bias=0), Conversion Mode=::NewEnumerator1)
     Direct Array Mode=::NewEnumerator1
     Interpolation=&Sine(Normalized Angle=$Emitter.Age, Period=0.5, Scale=1, Bias=0)
     Color Selection Array=DataInterface() {propertyValues {ColorData [<rgba 0.9 0.69 0 1>, <rgba 0.9 0.69 0 2>]}}
     Random Seed=?
    )
    R3 SkyBlue:LinearColor=&SelectLinearColorFromArray(
     Array Sampling Mode=::NewEnumerator2
     Direct Array Index=&MakeIntFromFloat(Float=&Sine(Normalized Angle=$Emitter.Age, Period=0.5, Scale=1, Bias=0), Conversion Mode=::NewEnumerator1)
     Direct Array Mode=::NewEnumerator1
     Interpolation=&Sine(Normalized Angle=$Emitter.Age, Period=0.5, Scale=1, Bias=0)
     Color Selection Array=DataInterface() {propertyValues {ColorData [<rgba 0.4062 1.4667 2 0.5>, <rgba 0.4062 1.4667 2 1.2>]}}
     Random Seed=?
    )
    R3 TitaniumWhite:LinearColor=&SelectLinearColorFromArray(
     Array Sampling Mode=::NewEnumerator2
     Direct Array Index=&MakeIntFromFloat(Float=&Sine(Normalized Angle=$Emitter.Age, Period=0.5, Scale=1, Bias=0), Conversion Mode=::NewEnumerator1)
     Direct Array Mode=::NewEnumerator1
     Interpolation=&Sine(Normalized Angle=$Emitter.Age, Period=0.5, Scale=1, Bias=0)
     Color Selection Array=DataInterface() {propertyValues {ColorData [<rgba 0.8906 0.8906 0.8906 1>, <rgba 1 1 1 2>]}}
     Random Seed=?
    )
    Saffron=<rgba 0.9141 0.872 0.3487 3>
    SkyBlue=<rgba 0.1831 0.8677 1.212 5>
    TitaniumWhite=<rgba 0.9948 0.9948 0.9948 4>
    Use R1 Painted Overrides=true
    Use R2 Painted Overrides=true
    Use R3 Painted Overrides=true
   )
   &VC_InitializeRibbon_v1(
    Color=$StackContext.Color
    Custom Offset=true
    Lifetime=1.15
    Local Vector Offset=&AddVectorToPosition(
     Position=<-70 0 0>
     Vector=&AddVectorToPosition(Position=$User.DTCarOffset, Vector=Linked() {linkedVariable {name=User.SingleBoostPos, type=NiagaraPosition}})
    )
    Offset Amount=1
    Ribbon Width=35
   )
   &AddVelocity(Velocity=&MakeCustomVectorFromBool(Bool=$User.bIsFrontendPreview, True Vector=<-1500 0 0>, False Vector=<-1000 0 0>)) {module=AddVelocity001}
  ]
  particleUpdateScript [
   &ParticleState()
   /SetVariables(
    Particles.R1ColorGlitterPink:LinearColor=&VC_DynamicScaleColor(
     Color to Scale:LinearColor=$Particles.R1 Color
     RGB Scale:Vector3f=<1 1 1>
     Alpha Scale:NiagaraFloat=&FloatFromCurve(FloatCurve=DataInterface() {propertyValues {Curve {keys [<at 0 0>,<at 0.5 1>,<at 0.6 0>]}}})
    )
    Particles.R1ScaleGlitterPink:NiagaraFloat=&Multiply_Float(
     A=$Particles.Initial.RibbonWidth
     B=&FloatFromCurve(FloatCurve=DataInterface() {propertyValues {Curve {keys [<at 0 1>,<at 0.4 4>,<at 1 9>]}}}, Scale Curve=0.7)
    )
    Particles.R2ColorGlitterBlue:LinearColor=&VC_DynamicScaleColor(
     Color to Scale:LinearColor=$Particles.R2 Color
     RGB Scale:Vector3f=<3 3 3>
     Alpha Scale:NiagaraFloat=&FloatFromCurve(FloatCurve=DataInterface() {propertyValues {Curve {keys [<at 0 0>,<at 0.2 1>,<at 0.9 0>]}}})
    )
    Particles.R2ScaleGlitterBlue:NiagaraFloat=&Multiply_Float(
     A=$Particles.Initial.RibbonWidth
     B=&FloatFromCurve(FloatCurve=DataInterface() {propertyValues {Curve {keys [<at 0 1>,<at 0.4 8>,<at 1 10>]}}}, Scale Curve=0.7)
    )
    Particles.RibbonBrightness:NiagaraFloat=&Multiply_Float(
     A=$Particles.Initial.RibbonWidth
     B=&FloatFromCurve(FloatCurve=DataInterface() {propertyValues {Curve {keys [<at 0 0.5>,<at 0.4 2.2>,<at 1 4.5>]}}}, Scale Curve=0.7)
    )
   ) {isSetParameters=true}
   &InheritSourceMovement(
    Applied Position Scale=&VectorFromFloat(Value=&FloatFromCurve(FloatCurve=DataInterface() {propertyValues {Curve {keys [
          {interpMode=RCIM_Cubic, tangentMode=RCTM_Break, time=0, value=1}
          {arriveTangent=-2.9025, interpMode=RCIM_Cubic, tangentMode=RCTM_Break, time=1, value=0}
         ]}}}))
    ApplyForce=false
    ApplyPosition=true
    ApplyPositionFalloff=false
   )
   &SolveForcesAndVelocity(Clamp Velocity=false, Limit Acceleration=false)
   &VC_UpdateRibbon_v1()
   &ScaleRibbonWidth(Ribbon Width Scale=&FloatFromCurve(FloatCurve=DataInterface() {propertyValues {Curve {keys [<at 0 1>,<at 0.4 3>,<at 1 8.1647>]}}}))
   &ScaleColor(
    Scale Alpha=&FloatFromCurve(FloatCurve=DataInterface() {propertyValues {Curve {keys [<at 0 0>,<at 0.4 1>,<at 0.85 0>]}}})
    Scale RGB=&VectorFromFloat(Value=&MakeCustomFloatFromBool(Bool=$User.bIsPainted, True Float=$User.R1_ScaleColor, False Float=1))
    ScaleA=true
    ScaleRGB=true
   )
  ]
  rendererClasses [&NiagaraRibbonRendererProperties]
  renderers [
   {rendererClass=&NiagaraRibbonRendererProperties}
   {rendererClass=&NiagaraRibbonRendererProperties, rendererIndex=1}
   {rendererClass=&NiagaraRibbonRendererProperties, rendererIndex=2}
   {rendererClass=&NiagaraRibbonRendererProperties, rendererIndex=3}
  ]
 }
}
summary {
 systemName=NS_Boost_Splitenergy
 emitters [
  {emitterName=NE_VC_Boost_Ribbon_Optimized, rendererClasses [&NiagaraRibbonRendererProperties]}
  {emitterName=NE_VC_BoostCone, rendererClasses [&NiagaraSpriteRendererProperties]}
 ]
 userVariables [
  User.bIsFrontendPreview {type=NiagaraBool}
  User.bIsPainted {type=NiagaraBool}
  User.ColorVariantID {type=NiagaraInt32}
  User.DTCarOffset {type=NiagaraPosition, description="", defaultValue {}}
  User.PaintedColor {type=LinearColor, description="", defaultValue {struct=LinearColor, value=<rgba 1 1 1 1>}}
  User.PaintedColorR2 {type=LinearColor, description="", defaultValue {struct=LinearColor, value=<rgba 1 1 1 1>}}
  User.PaintedScalar {type=NiagaraFloat, defaultValue=1, description=""}
  User.PortCount {type=NiagaraInt32, defaultValue=1}
  User.R1_ScaleColor {type=NiagaraFloat, defaultValue=1, description=""}
  User.R2_ScaleColor {type=NiagaraFloat, defaultValue=1, description=""}
  User.R3_ScaleColor {type=NiagaraFloat, defaultValue=1, description=""}
  User.SingleBoostPos {type=NiagaraPosition, description="", defaultValue {}}
  User.StarsPaint {type=NiagaraFloat, defaultValue=1, description=""}
  User.WingsBorderColor {type=LinearColor, description="", defaultValue {struct=LinearColor, value=<rgba 1 1 1 1>}}
  User.WingsColor {type=LinearColor, description="", defaultValue {struct=LinearColor, value=<rgba 1 1 1 1>}}
 ]
}
